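# Sound cue for the Baba Yaga glider opening, layering a deploy sound with a witch cackle.
# A quality-level switch sends its first slot to a mix of the open wave and a half-second-delayed random pick from seven laughs.
# The other three slots share a lighter mix with two undelayed laugh options.
# It uses shared glider open/close attenuation and concurrency, plus the glider sound class.
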SoundCue /BRCosmetics/Sounds/Gliders/BabaYaga/Glider_BabaYaga_Open
root=QualityLevel_0
nodes [
 {id=Delay_0, children [Random_0], properties {delayMax=0.5, delayMin=0.5}}
 {id=Mixer_0, children [WavePlayer_1,Delay_0], properties {inputVolume [1,1]}}
 {id=Mixer_1, children [WavePlayer_11, Random_1], properties {inputVolume [1,1]}}
 {id=QualityLevel_0, children [Mixer_0,Mixer_1,Mixer_1,Mixer_1]}
 {
  id=Random_0
  children [WavePlayer_2,WavePlayer_3,WavePlayer_4,WavePlayer_5,WavePlayer_6,WavePlayer_7,WavePlayer_8]
  properties {weights [1,1,1,1,1,1,1]}
 }
 {id=Random_1, children [WavePlayer_9, WavePlayer_10], properties {weights [1,1]}}
 {id=WavePlayer_1, properties {soundWaveAssetPtr=&/Glider_BabaYaga_Open_01}}
 {id=WavePlayer_10, properties {soundWaveAssetPtr=&/Glider_BabaYaga_WitchLaugh_02}}
 {id=WavePlayer_11, properties {soundWaveAssetPtr=&Glider_BabaYaga_Open_01}}
 {id=WavePlayer_2, properties {soundWaveAssetPtr=&/Glider_BabaYaga_WitchLaugh_01}}
 {id=WavePlayer_3, properties {soundWaveAssetPtr=&Glider_BabaYaga_WitchLaugh_02}}
 {id=WavePlayer_4, properties {soundWaveAssetPtr=&/Glider_BabaYaga_WitchLaugh_03}}
 {id=WavePlayer_5, properties {soundWaveAssetPtr=&/Glider_BabaYaga_WitchLaugh_04}}
 {id=WavePlayer_6, properties {soundWaveAssetPtr=&/Glider_BabaYaga_WitchLaugh_05}}
 {id=WavePlayer_7, properties {soundWaveAssetPtr=&/Glider_BabaYaga_WitchLaugh_06}}
 {id=WavePlayer_8, properties {soundWaveAssetPtr=&/Glider_BabaYaga_WitchLaugh_07}}
 {id=WavePlayer_9, properties {soundWaveAssetPtr=&Glider_BabaYaga_WitchLaugh_04}}
]
properties {
 attenuationSettings=/Game/Sounds/Attenuation/Fort_Player_Glider_OpenClose_Athena_Att
 audioPropertiesSheet=/BRCosmetics/Sounds/Settings/PropertySheets/Glider/APS_Cosmetic_Glider_Open
 maxDistance=10000
 soundClassObject=/Game/Sounds/FortSoundClassesAndModes/Fort_Gliders
 concurrencySet [&Athena_GliderOpenClose_Conc]
}
